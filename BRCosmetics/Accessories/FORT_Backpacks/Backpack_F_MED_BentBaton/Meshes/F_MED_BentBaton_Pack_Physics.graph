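# Physics asset for the BentBaton backpack cosmetic, giving the pack and its two dangling-strap bone chains capsule bodies for physics simulation.
# Left and right strap chains hang from base bodies whose angular motion is locked to the large pack capsule.
# Joints along each chain lock selected swing and twist axes, restricting how the straps bend.
# Two mirrored, angularly locked constraints tie the last left and right bodies together, closing the straps into a loop.
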PhysicsAsset /BRCosmetics/Accessories/FORT_Backpacks/Backpack_F_MED_BentBaton/Meshes/F_MED_BentBaton_Pack_Physics
bodies {
 dyn_strap_L_00 [{
   shapeType=Capsule
   length=1.9049
   radius=1.4335
   shapeName=dyn_strap_L_00_capsule
   center=<0.5288 0.078 -0.3537>
   rotation=<rot 0 178.764 -83.6964>
  }]
 dyn_strap_L_01 [{
   shapeType=Capsule
   length=5.1447
   radius=1.2128
   shapeName=dyn_strap_L_01_capsule
   center=<2.8676 -0.0962 -0.6572>
   rotation=<rot -0.0001 -94.1397 -85.2125>
  }]
 dyn_strap_L_02 [{
   shapeType=Capsule
   length=3.3747
   radius=2.1987
   shapeName=dyn_strap_L_02_capsule
   center=<1.6595 -0.035 -0.52>
   rotation=<rot 0 -86.9685 -91.3686>
  }]
 dyn_strap_L_03 [{
   shapeType=Capsule
   length=8.8596
   radius=1.095
   shapeName=dyn_strap_L_03_capsule
   center=<3.2816 0.0212 -0.0121>
   rotation=<rot 0 -89.2166 -90.3014>
  }]
 dyn_strap_L_04 [{
   shapeType=Capsule
   length=8.4953
   radius=1.588
   shapeName=dyn_strap_L_04_capsule
   center=<4.9658 0.1117 0.0375>
   rotation=<rot 0 -88.4439 -87.5803>
  }]
 dyn_strap_L_05 [{
   shapeType=Capsule
   length=3.9355
   radius=1.6317
   shapeName=dyn_strap_L_05_capsule
   center=<2.3515 0.2947 -0.4019>
   rotation=<rot 34.2737 -79.4534 -82.3109>
  }]
 dyn_strap_L_base [{shapeType=Capsule, length=0, radius=0.2, shapeName=dyn_strap_L_base_capsule}]
 dyn_strap_R_00 [{
   shapeType=Capsule
   length=1.9105
   radius=1.3131
   shapeName=dyn_strap_R_00_capsule
   center=<1.0068 0.1081 -0.0473>
   rotation=<rot 0 0.4131 -83.849>
  }]
 dyn_strap_R_01 [{
   shapeType=Capsule
   length=5.4887
   radius=1.2075
   shapeName=dyn_strap_R_01_capsule
   center=<3.52 0.2169 -0.1395>
   rotation=<rot 0 -89.1107 -87.8214>
  }]
 dyn_strap_R_02 [{
   shapeType=Capsule
   length=5.8823
   radius=1.2134
   shapeName=dyn_strap_R_02_capsule
   center=<3.4818 -0.0359 -0.458>
   rotation=<rot 0 86.8213 -88.1623>
  }]
 dyn_strap_R_03 [{
   shapeType=Capsule
   length=5.8287
   radius=1.5479
   shapeName=dyn_strap_R_03_capsule
   center=<2.0782 -0.0846 -0.2254>
   rotation=<rot 0 -96.9488 -79.4546>
  }]
 dyn_strap_R_base [{shapeType=Capsule, length=0, radius=0.2, shapeName=dyn_strap_R_base_capsule}]
 pack [{shapeType=Capsule, length=17.8513, radius=9.7939, center=<0.146 -8.1029 -7.1443>, rotation=<rot -37.9534 0.0168 -0.0273>}]
}
constraints [
 {
  bone1Name=dyn_strap_L_base
  bone2Name=pack
  swing1LimitDegrees=10
  swing1Motion=Locked
  swing2LimitDegrees=5
  swing2Motion=Locked
  twistLimitDegrees=2
  twistMotion=Locked
 }
 {bone1Name=dyn_strap_L_00, bone2Name=dyn_strap_L_base, swing1LimitDegrees=15, swing2LimitDegrees=5, twistLimitDegrees=2}
 {bone1Name=dyn_strap_L_01, bone2Name=dyn_strap_L_00, swing1LimitDegrees=20, swing2LimitDegrees=5, twistLimitDegrees=2}
 {bone1Name=dyn_strap_L_02, bone2Name=dyn_strap_L_01, swing1LimitDegrees=15, swing1Motion=Locked, swing2LimitDegrees=5, twistLimitDegrees=2}
 {bone1Name=dyn_strap_L_03, bone2Name=dyn_strap_L_02, swing1LimitDegrees=5, swing1Motion=Locked, swing2LimitDegrees=2, twistLimitDegrees=2}
 {bone1Name=dyn_strap_L_04, bone2Name=dyn_strap_L_03, swing1LimitDegrees=5, swing1Motion=Locked, swing2LimitDegrees=5, twistLimitDegrees=2}
 {
  bone1Name=dyn_strap_L_05
  bone2Name=dyn_strap_L_04
  swing1LimitDegrees=10
  swing1Motion=Locked
  swing2LimitDegrees=5
  swing2Motion=Locked
  twistLimitDegrees=2
  twistMotion=Locked
 }
 {
  bone1Name=dyn_strap_R_base
  bone2Name=pack
  swing1LimitDegrees=10
  swing1Motion=Locked
  swing2LimitDegrees=0
  swing2Motion=Locked
  twistLimitDegrees=0
  twistMotion=Locked
 }
 {bone1Name=dyn_strap_R_00, bone2Name=dyn_strap_R_base, swing1LimitDegrees=25, swing2LimitDegrees=5, twistLimitDegrees=2}
 {bone1Name=dyn_strap_R_01, bone2Name=dyn_strap_R_00, swing1LimitDegrees=20, swing2LimitDegrees=0, twistLimitDegrees=2}
 {
  bone1Name=dyn_strap_R_02
  bone2Name=dyn_strap_R_01
  swing1LimitDegrees=10
  swing1Motion=Locked
  swing2LimitDegrees=5
  swing2Motion=Locked
  twistLimitDegrees=2
 }
 {
  bone1Name=dyn_strap_R_03
  bone2Name=dyn_strap_R_02
  swing1LimitDegrees=10
  swing1Motion=Locked
  swing2LimitDegrees=5
  swing2Motion=Locked
  twistLimitDegrees=2
  twistMotion=Locked
 }
 {
  bone1Name=dyn_strap_R_03
  bone2Name=dyn_strap_L_05
  swing1LimitDegrees=10
  swing1Motion=Locked
  swing2LimitDegrees=5
  swing2Motion=Locked
  twistLimitDegrees=2
  twistMotion=Locked
 }
 {
  bone1Name=dyn_strap_L_05
  bone2Name=dyn_strap_R_03
  swing1LimitDegrees=10
  swing1Motion=Locked
  swing2LimitDegrees=5
  swing2Motion=Locked
  twistLimitDegrees=2
  twistMotion=Locked
 }
]
