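# Animation Blueprint for the BentBaton backpack cosmetic, adding secondary motion to its pompoms, straps, handles and zippers.
# With rigid bodies enabled, RigidBody is combined with AnimDynamics, using a separate world-space pompom setup in the front end.
# Otherwise AnimDynamics drives everything, with bone constraints linking the straps and their simulation weight halved while skydiving.
# Zipper dynamics follow both paths, then blended additive bone offsets re-pose pack and straps for skydive states.
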AnimBlueprint /BRCosmetics/Accessories/FORT_Backpacks/Backpack_F_MED_BentBaton/Meshes/F_MED_BentBaton_Pack_AnimBP
target_skeleton=&/F_MED_BentBaton_Pack_Skeleton
graphs [
 AnimGraph {nodes [
   n1 = Anim Dynamics - Bone: dyn_pompom_lower_00(ComponentPose=Local To Component(LocalPose=Use cached pose 'input pose'()), Alpha=1, ExternalForce="0,0,0") {
    angularDampingOverride=0.999
    angularSpringConstant=800
    bAngularSpring=true
    bUseGravityOverride=true
    alphaBoolBlend {blendOption=Linear}
    boundBone {boneName=dyn_pompom_lower_00}
    componentAppliedLinearAccClamp=<1000 1000 1200>
    componentLinearAccScale=<0.5 0.5 0.5>
    componentLinearVelScale=<0.5 0.5 0.5>
    gravityOverride {z=-800}
    physicsBodyDefinitions [{
      boundBone {boneName=dyn_pompom_lower_00}
      boxExtents=<10 10 10>
      constraintSetup {angularLimitsMax=<20 30 0>, angularLimitsMin=<-20 0 0>, angularTarget=<0 0 0>}
      localJointOffset=<1 0 0>
     }]
   }
   n2 = Anim Dynamics - Bone: dyn_pompom_lower_01(ComponentPose=@n1, Alpha=1, ExternalForce="0,0,0") {
    angularDampingOverride=0.999
    angularSpringConstant=900
    bUseGravityOverride=true
    alphaBoolBlend {blendOption=Linear}
    boundBone {boneName=dyn_pompom_lower_01}
    componentAppliedLinearAccClamp=<900 900 900>
    componentLinearAccScale=<0.4 0.4 0.4>
    componentLinearVelScale=<0.1 0.1 0.1>
    gravityOverride {z=-980}
    physicsBodyDefinitions [{
      boundBone {boneName=dyn_pompom_lower_01}
      boxExtents=<10 10 10>
      constraintSetup {angularLimitsMax=<20 20 5>, angularLimitsMin=<-20 0 0>, angularTarget=<0 0 0>}
      localJointOffset=<0.2 0 0>
     }]
   }
   n3 = Anim Dynamics - Bone: dyn_pompom_lower_02(ComponentPose=@n2, Alpha=1, ExternalForce="0,0,0") {
    angularDampingOverride=0.999
    angularSpringConstant=900
    bAngularSpring=true
    bUseGravityOverride=true
    alphaBoolBlend {blendOption=Linear}
    boundBone {boneName=dyn_pompom_lower_02}
    componentAppliedLinearAccClamp=<900 900 900>
    componentLinearAccScale=<0.4 0.4 0.4>
    componentLinearVelScale=<0.1 0.1 0.1>
    gravityOverride {z=-980}
    physicsBodyDefinitions [{
      boundBone {boneName=dyn_pompom_lower_02}
      boxExtents=<10 10 10>
      constraintSetup {angularLimitsMax=<20 15 0>, angularLimitsMin=<-20 -5 -5>, angularTarget=<0 0 0>}
      localJointOffset=<-0.2 0 0>
     }]
   }
   n4 = Anim Dynamics - Bone: dyn_pompom_middle_00(ComponentPose=@n3, Alpha=1, ExternalForce="0,0,0") {
    angularDampingOverride=0.999
    angularSpringConstant=900
    bUseGravityOverride=true
    alphaBoolBlend {blendOption=Linear}
    boundBone {boneName=dyn_pompom_middle_00}
    componentAppliedLinearAccClamp=<900 900 900>
    componentLinearAccScale=<0.4 0.4 0.4>
    componentLinearVelScale=<0.1 0.1 0.1>
    gravityOverride {z=-800}
    physicsBodyDefinitions [{
      boundBone {boneName=dyn_pompom_middle_00}
      boxExtents=<7 7 7>
      constraintSetup {angularLimitsMax=<20 30 0>, angularLimitsMin=<-20 0 -10>, angularTarget=<0 0 0>}
      localJointOffset=<0.3 0 0>
     }]
   }
   n5 = Anim Dynamics - Bone: dyn_pompom_middle_01(ComponentPose=@n4, Alpha=1, ExternalForce="0,0,0") {
    angularDampingOverride=0.999
    angularSpringConstant=900
    bUseGravityOverride=true
    alphaBoolBlend {blendOption=Linear}
    boundBone {boneName=dyn_pompom_middle_01}
    componentAppliedLinearAccClamp=<900 900 900>
    componentLinearAccScale=<0.4 0.4 0.4>
    componentLinearVelScale=<0.1 0.1 0.1>
    gravityOverride {z=-800}
    physicsBodyDefinitions [{
      boundBone {boneName=dyn_pompom_middle_01}
      boxExtents=<7 7 7>
      constraintSetup {angularLimitsMax=<20 25 5>, angularLimitsMin=<-20 0 0>, angularTarget=<0 0 0>}
      localJointOffset=<0.3 0 0>
     }]
   }
   n6 = Anim Dynamics - Bone: dyn_pompom_middle_02(ComponentPose=@n5, Alpha=1, ExternalForce="0,0,0") {
    angularDampingOverride=0.999
    angularSpringConstant=800
    bAngularSpring=true
    bUseGravityOverride=true
    alphaBoolBlend {blendOption=Linear}
    boundBone {boneName=dyn_pompom_middle_02}
    componentAppliedLinearAccClamp=<900 900 900>
    componentLinearAccScale=<0.4 0.4 0.4>
    componentLinearVelScale=<0.1 0.1 0.1>
    gravityOverride {z=-800}
    physicsBodyDefinitions [{
      boundBone {boneName=dyn_pompom_middle_02}
      boxExtents=<7 7 7>
      constraintSetup {angularLimitsMax=<20 15 5>, angularLimitsMin=<-20 0 0>, angularTarget=<0 0 0>}
      localJointOffset=<-0.1 0 0>
     }]
   }
   n7 = Anim Dynamics - Bone: dyn_pompom_top_00(ComponentPose=@n6, Alpha=1, ExternalForce="0,0,0") {
    angularDampingOverride=0.999
    angularSpringConstant=800
    bUseGravityOverride=true
    alphaBoolBlend {blendOption=Linear}
    boundBone {boneName=dyn_pompom_top_00}
    componentAppliedLinearAccClamp=<900 900 900>
    componentLinearAccScale=<0.4 0.4 0.4>
    componentLinearVelScale=<0.1 0.1 0.1>
    gravityOverride {z=-750}
    physicsBodyDefinitions [{
      boundBone {boneName=dyn_pompom_top_00}
      boxExtents=<5 5 5>
      constraintSetup {angularLimitsMax=<20 40 0>, angularLimitsMin=<-20 0 -10>, angularTarget=<1 0 0>}
      localJointOffset=<0.5 0 0>
     }]
   }
   n8 = Anim Dynamics - Bone: dyn_pompom_top_01(ComponentPose=@n7, Alpha=1, ExternalForce="0,0,0") {
    angularDampingOverride=0.999
    angularSpringConstant=800
    bOverrideAngularDamping=true
    bUseGravityOverride=true
    alphaBoolBlend {blendOption=Linear}
    boundBone {boneName=dyn_pompom_top_01}
    componentAppliedLinearAccClamp=<900 900 900>
    componentLinearAccScale=<0.4 0.4 0.4>
    componentLinearVelScale=<0.1 0.1 0.1>
    gravityOverride {z=-750}
    physicsBodyDefinitions [{
      boundBone {boneName=dyn_pompom_top_01}
      boxExtents=<5 5 5>
      constraintSetup {angularLimitsMax=<20 15 0>, angularLimitsMin=<-20 -5 -5>, angularTarget=<1 0 0>}
      localJointOffset=<0.1 0 1>
     }]
    relativeSpaceBone {boneName=pack}
   }
   n9 = Anim Dynamics - Bone: dyn_pompom_top_02(ComponentPose=@n8, Alpha=1, ExternalForce="0,0,0") {
    angularDampingOverride=1
    angularSpringConstant=5000
    bAngularSpring=true
    bOverrideAngularDamping=true
    bUseGravityOverride=true
    alphaBoolBlend {blendOption=Linear}
    boundBone {boneName=dyn_pompom_top_02}
    componentAppliedLinearAccClamp=<600 600 600>
    componentLinearAccScale=<0.2 0.2 0.2>
    componentLinearVelScale=<0.3 0.3 0.3>
    gravityOverride {z=-980}
    physicsBodyDefinitions [{
      boundBone {boneName=dyn_pompom_top_02}
      boxExtents=<6 6 6>
      constraintSetup {angularTargetAxis=AxisZ, angularLimitsMax=<20 50 0>, angularLimitsMin=<-20 -30 0>, angularTarget=<0 0 1>}
      localJointOffset=<-0.2 0 0.1>
     }]
    relativeSpaceBone {boneName=dyn_pompom_top_02}
   }
   n10 = RigidBody!(ComponentPose=@n9, Alpha, OverrideWorldGravity:Vector="0,0,-980", ExternalForce) -> (Pose) {
    bOverrideWorldGravity=true
    overridePhysicsAsset=&/F_MED_BentBaton_Pack_Physics
    alphaBoolBlend {blendOption=Linear}
    overrideWorldGravity {z=-980}
    simSpaceSettings {
     maxAngularAcceleration=15
     maxAngularVelocity=3
     maxLinearAcceleration=800
     maxLinearVelocity=500
     velocityScaleZ=0.35
     worldAlpha=1
     externalLinearDragV=<0.1 0.1 0.1>
    }
   }
   n11 = Anim Dynamics - Bone: dyn_handles_00(ComponentPose=@n10, Alpha=1, ExternalForce="0,0,0") {
    angularDampingOverride=1
    angularSpringConstant=3000
    bAngularSpring=true
    bOverrideAngularDamping=true
    bUseGravityOverride=true
    bUseSphericalLimits=true
    alphaBoolBlend {blendOption=Linear}
    boundBone {boneName=dyn_handles_00}
    componentAppliedLinearAccClamp=<1000 1000 1000>
    componentLinearAccScale=<0.3 0.3 0.3>
    componentLinearVelScale=<0.4 0.4 0.4>
    gravityOverride {z=-980}
    physicsBodyDefinitions [{
      boundBone {boneName=dyn_handles_00}
      boxExtents=<7 7 35>
      constraintSetup {angularLimitsMax=<5 0 15>, angularLimitsMin=<-5 -5 -10>, angularTarget=<1 0 0>}
      localJointOffset=<0.5 0 0.5>
     }]
    sphericalLimits [{limitRadius=5, drivingBone {boneName=pack}, sphereLocalOffset=<-13.7344 -9.1072 -5.8926>}]
   }
   n12 = Anim Dynamics - Bone: dyn_handles_01(ComponentPose=@n11, Alpha=1, ExternalForce="0,0,0") {
    angularDampingOverride=1
    angularSpringConstant=3000
    bAngularSpring=true
    bOverrideAngularDamping=true
    bUseGravityOverride=true
    bUseSphericalLimits=true
    alphaBoolBlend {blendOption=Linear}
    boundBone {boneName=dyn_handles_01}
    componentAppliedLinearAccClamp=<1000 1000 1000>
    componentLinearAccScale=<0.3 0.3 0.3>
    componentLinearVelScale=<0.4 0.4 0.4>
    gravityOverride {z=-980}
    physicsBodyDefinitions [{
      boundBone {boneName=dyn_handles_01}
      boxExtents=<7 7 20>
      constraintSetup {angularLimitsMax=<0 0 5>, angularLimitsMin=<-2 0 -5>, angularTarget=<1 0 0>}
      localJointOffset=<1 0 0>
     }]
    sphericalLimits [{limitRadius=5, drivingBone {boneName=pack}, sphereLocalOffset=<-15.4322 -8.9155 -8.5242>}]
   }
   n13 = RigidBody(ComponentPose=Local To Component(LocalPose=Use cached pose 'FE fix'()), OverrideWorldGravity="0,100,-980") {
    bOverrideWorldGravity=true
    overridePhysicsAsset=&F_MED_BentBaton_Pack_Physics
    alphaBoolBlend {blendOption=Linear}
    overrideWorldGravity {z=-980}
    simSpaceSettings {
     maxAngularAcceleration=15
     maxAngularVelocity=3
     maxLinearAcceleration=800
     maxLinearVelocity=500
     worldAlpha=1
     externalLinearDragV=<0.1 0.1 0.1>
    }
   }
   n14 = "Transform (Modify) Bone - Bone: dyn_strap_L_02"(ComponentPose=@n13, Alpha=1, Translation="0,0,0", Rotation="0,-5,0", Scale="1,1,1") {
    rotationMode=BMM_Additive
    rotationSpace=BCS_BoneSpace
    translationMode=BMM_Additive
    translationSpace=BCS_BoneSpace
    alphaBoolBlend {blendOption=Linear}
    boneToModify {boneName=dyn_strap_L_02}
    rotation {yaw=-5}
   }
   n15 = "Transform (Modify) Bone - Bone: dyn_strap_R_01"(ComponentPose=@n14, Alpha=1, Translation="0,0,0", Rotation="0,5,0", Scale="1,1,1") {
    rotationMode=BMM_Additive
    rotationSpace=BCS_BoneSpace
    translationMode=BMM_Additive
    translationSpace=BCS_BoneSpace
    alphaBoolBlend {blendOption=Linear}
    boneToModify {boneName=dyn_strap_R_01}
    rotation {yaw=5}
   }
   n16 = Anim Dynamics - Bone: dyn_handles_01(ComponentPose=@n15, Alpha=1, ExternalForce="0,0,0") {
    angularDampingOverride=1
    angularSpringConstant=10000
    bAngularSpring=true
    bOverrideAngularDamping=true
    bUseGravityOverride=true
    bUseSphericalLimits=true
    simulationSpace=World
    alphaBoolBlend {blendOption=Linear}
    boundBone {boneName=dyn_handles_01}
    componentAppliedLinearAccClamp=<1000 1000 1000>
    componentLinearAccScale=<0.6 0.6 0.6>
    componentLinearVelScale=<0.6 0.6 0.6>
    gravityOverride {z=-980}
    physicsBodyDefinitions [{
      boundBone {boneName=dyn_handles_01}
      boxExtents=<50 50 50>
      constraintSetup {angularLimitsMax=<0 0 30>, angularLimitsMin=<-2 0 -30>, angularTarget=<1 0 0>}
      localJointOffset=<1 0 0>
     }]
    relativeSpaceBone {boneName=pack}
    sphericalLimits [{limitRadius=5, drivingBone {boneName=pack}, sphereLocalOffset=<-15.4322 -8.9155 -8.5242>}]
   }
   n17 = Blend Poses by bool(BlendPose_0=Component To Local(ComponentPose=@n16), BlendPose_1=Component To Local(ComponentPose=@n12), bActiveValue=Get bIsInFrontEnd()) {blendType=Linear}
   n18 = Anim Dynamics - Bone: dyn_handles_00(ComponentPose=Local To Component(LocalPose=Use cached pose 'input pose'()), Alpha=1, ExternalForce="0,0,0") {
    angularDampingOverride=1
    angularSpringConstant=3000
    bAngularSpring=true
    bOverrideAngularDamping=true
    bUseGravityOverride=true
    bUseSphericalLimits=true
    alphaBoolBlend {blendOption=Linear}
    boundBone {boneName=dyn_handles_00}
    componentAppliedLinearAccClamp=<1000 1000 1000>
    componentLinearAccScale=<0.3 0.3 0.3>
    componentLinearVelScale=<0.4 0.4 0.4>
    gravityOverride {z=-980}
    physicsBodyDefinitions [{
      boundBone {boneName=dyn_handles_00}
      boxExtents=<7 7 35>
      constraintSetup {angularLimitsMax=<5 0 15>, angularLimitsMin=<-5 -5 -10>, angularTarget=<1 0 0>}
      localJointOffset=<0.5 0 0.5>
     }]
    sphericalLimits [{limitRadius=5, drivingBone {boneName=pack}, sphereLocalOffset=<-13.7344 -9.1072 -5.8926>}]
   }
   n19 = Anim Dynamics - Bone: dyn_handles_01(ComponentPose=@n18, Alpha=1, ExternalForce="0,0,0") {
    angularDampingOverride=1
    angularSpringConstant=3000
    bAngularSpring=true
    bOverrideAngularDamping=true
    bUseGravityOverride=true
    bUseSphericalLimits=true
    alphaBoolBlend {blendOption=Linear}
    boundBone {boneName=dyn_handles_01}
    componentAppliedLinearAccClamp=<1000 1000 1000>
    componentLinearAccScale=<0.3 0.3 0.3>
    componentLinearVelScale=<0.4 0.4 0.4>
    gravityOverride {z=-980}
    physicsBodyDefinitions [{
      boundBone {boneName=dyn_handles_01}
      boxExtents=<7 7 20>
      constraintSetup {angularLimitsMax=<0 0 5>, angularLimitsMin=<-2 0 -5>, angularTarget=<1 0 0>}
      localJointOffset=<1 0 0>
     }]
    sphericalLimits [{limitRadius=5, drivingBone {boneName=pack}, sphereLocalOffset=<-15.4322 -8.9155 -8.5242>}]
   }
   n20 = Anim Dynamics - Bone: dyn_pompom_lower_00(ComponentPose=@n19, Alpha=1, ExternalForce="0,0,0") {
    angularDampingOverride=0.999
    angularSpringConstant=800
    bAngularSpring=true
    bUseGravityOverride=true
    alphaBoolBlend {blendOption=Linear}
    boundBone {boneName=dyn_pompom_lower_00}
    componentAppliedLinearAccClamp=<1000 1000 1200>
    componentLinearAccScale=<0.5 0.5 0.5>
    componentLinearVelScale=<0.5 0.5 0.5>
    gravityOverride {z=-800}
    physicsBodyDefinitions [{
      boundBone {boneName=dyn_pompom_lower_00}
      boxExtents=<10 10 10>
      constraintSetup {angularLimitsMax=<20 30 0>, angularLimitsMin=<-20 0 0>, angularTarget=<0 0 0>}
      localJointOffset=<1 0 0>
     }]
   }
   n21 = Anim Dynamics - Bone: dyn_pompom_lower_01(ComponentPose=@n20, Alpha=1, ExternalForce="0,0,0") {
    angularDampingOverride=0.999
    angularSpringConstant=900
    bUseGravityOverride=true
    alphaBoolBlend {blendOption=Linear}
    boundBone {boneName=dyn_pompom_lower_01}
    componentAppliedLinearAccClamp=<900 900 900>
    componentLinearAccScale=<0.4 0.4 0.4>
    componentLinearVelScale=<0.1 0.1 0.1>
    gravityOverride {z=-980}
    physicsBodyDefinitions [{
      boundBone {boneName=dyn_pompom_lower_01}
      boxExtents=<10 10 10>
      constraintSetup {angularLimitsMax=<20 20 5>, angularLimitsMin=<-20 0 0>, angularTarget=<0 0 0>}
      localJointOffset=<0.2 0 0>
     }]
   }
   n22 = Anim Dynamics - Bone: dyn_pompom_lower_02(ComponentPose=@n21, Alpha=1, ExternalForce="0,0,0") {
    angularDampingOverride=0.999
    angularSpringConstant=900
    bAngularSpring=true
    bUseGravityOverride=true
    alphaBoolBlend {blendOption=Linear}
    boundBone {boneName=dyn_pompom_lower_02}
    componentAppliedLinearAccClamp=<900 900 900>
    componentLinearAccScale=<0.4 0.4 0.4>
    componentLinearVelScale=<0.1 0.1 0.1>
    gravityOverride {z=-980}
    physicsBodyDefinitions [{
      boundBone {boneName=dyn_pompom_lower_02}
      boxExtents=<10 10 10>
      constraintSetup {angularLimitsMax=<20 15 0>, angularLimitsMin=<-20 -5 -5>, angularTarget=<0 0 0>}
      localJointOffset=<-0.2 0 0>
     }]
   }
   n23 = Anim Dynamics - Bone: dyn_pompom_middle_00(ComponentPose=@n22, Alpha=1, ExternalForce="0,0,0") {
    angularDampingOverride=0.999
    angularSpringConstant=900
    bUseGravityOverride=true
    alphaBoolBlend {blendOption=Linear}
    boundBone {boneName=dyn_pompom_middle_00}
    componentAppliedLinearAccClamp=<900 900 900>
    componentLinearAccScale=<0.4 0.4 0.4>
    componentLinearVelScale=<0.1 0.1 0.1>
    gravityOverride {z=-800}
    physicsBodyDefinitions [{
      boundBone {boneName=dyn_pompom_middle_00}
      boxExtents=<7 7 7>
      constraintSetup {angularLimitsMax=<20 30 0>, angularLimitsMin=<-20 0 -10>, angularTarget=<0 0 0>}
      localJointOffset=<0.3 0 0>
     }]
   }
   n24 = Anim Dynamics - Bone: dyn_pompom_middle_01(ComponentPose=@n23, Alpha=1, ExternalForce="0,0,0") {
    angularDampingOverride=0.999
    angularSpringConstant=900
    bUseGravityOverride=true
    alphaBoolBlend {blendOption=Linear}
    boundBone {boneName=dyn_pompom_middle_01}
    componentAppliedLinearAccClamp=<900 900 900>
    componentLinearAccScale=<0.4 0.4 0.4>
    componentLinearVelScale=<0.1 0.1 0.1>
    gravityOverride {z=-800}
    physicsBodyDefinitions [{
      boundBone {boneName=dyn_pompom_middle_01}
      boxExtents=<7 7 7>
      constraintSetup {angularLimitsMax=<20 25 5>, angularLimitsMin=<-20 0 0>, angularTarget=<0 0 0>}
      localJointOffset=<0.3 0 0>
     }]
   }
   n25 = Anim Dynamics - Bone: dyn_pompom_middle_02(ComponentPose=@n24, Alpha=1, ExternalForce="0,0,0") {
    angularDampingOverride=0.999
    angularSpringConstant=800
    bAngularSpring=true
    bUseGravityOverride=true
    alphaBoolBlend {blendOption=Linear}
    boundBone {boneName=dyn_pompom_middle_02}
    componentAppliedLinearAccClamp=<900 900 900>
    componentLinearAccScale=<0.4 0.4 0.4>
    componentLinearVelScale=<0.1 0.1 0.1>
    gravityOverride {z=-800}
    physicsBodyDefinitions [{
      boundBone {boneName=dyn_pompom_middle_02}
      boxExtents=<7 7 7>
      constraintSetup {angularLimitsMax=<20 15 5>, angularLimitsMin=<-20 0 0>, angularTarget=<0 0 0>}
      localJointOffset=<-0.1 0 0>
     }]
   }
   n26 = Anim Dynamics - Bone: dyn_pompom_top_00(ComponentPose=@n25, Alpha=1, ExternalForce="0,0,0") {
    angularDampingOverride=0.999
    angularSpringConstant=800
    bUseGravityOverride=true
    alphaBoolBlend {blendOption=Linear}
    boundBone {boneName=dyn_pompom_top_00}
    componentAppliedLinearAccClamp=<900 900 900>
    componentLinearAccScale=<0.4 0.4 0.4>
    componentLinearVelScale=<0.1 0.1 0.1>
    gravityOverride {z=-750}
    physicsBodyDefinitions [{
      boundBone {boneName=dyn_pompom_top_00}
      boxExtents=<5 5 5>
      constraintSetup {angularLimitsMax=<20 40 0>, angularLimitsMin=<-20 0 -10>, angularTarget=<0 0 0>}
      localJointOffset=<0.5 0 0>
     }]
   }
   n27 = Anim Dynamics - Bone: dyn_pompom_top_01(ComponentPose=@n26, Alpha=1, ExternalForce="0,0,0") {
    angularDampingOverride=0.999
    angularSpringConstant=800
    bOverrideAngularDamping=true
    bUseGravityOverride=true
    alphaBoolBlend {blendOption=Linear}
    boundBone {boneName=dyn_pompom_top_01}
    componentAppliedLinearAccClamp=<900 900 900>
    componentLinearAccScale=<0.4 0.4 0.4>
    componentLinearVelScale=<0.1 0.1 0.1>
    gravityOverride {z=-750}
    physicsBodyDefinitions [{
      boundBone {boneName=dyn_pompom_top_01}
      boxExtents=<5 5 5>
      constraintSetup {angularLimitsMax=<20 15 0>, angularLimitsMin=<-20 -5 -5>, angularTarget=<0 0 0>}
      localJointOffset=<0.1 0 1>
     }]
    relativeSpaceBone {boneName=pack}
   }
   n28 = Anim Dynamics - Bone: dyn_pompom_top_02(ComponentPose=@n27, Alpha=1, ExternalForce="0,0,0") {
    angularDampingOverride=0.999
    angularSpringConstant=800
    bAngularSpring=true
    bUseGravityOverride=true
    alphaBoolBlend {blendOption=Linear}
    boundBone {boneName=dyn_pompom_top_02}
    componentAppliedLinearAccClamp=<900 900 900>
    componentLinearAccScale=<0.4 0.4 0.4>
    componentLinearVelScale=<0.1 0.1 0.1>
    gravityOverride {z=-750}
    physicsBodyDefinitions [{
      boundBone {boneName=dyn_pompom_top_02}
      boxExtents=<5 5 5>
      constraintSetup {angularLimitsMax=<20 10 0>, angularLimitsMin=<-20 -30 -10>, angularTarget=<0 0 0>}
      localJointOffset=<-0.5 0 0>
     }]
   }
   n29 = Get scalability alpha() -> (scalability alpha:double)
   n30 = Anim Dynamics - Bone: dyn_strap_L_02(ComponentPose=@n28, Alpha=@n29, ExternalForce="0,0,0") {
    angularSpringConstant=1000
    bAngularSpring=true
    bUseGravityOverride=true
    alphaBoolBlend {blendOption=Linear}
    alphaScaleBiasClamp {bInterpResult=true, interpSpeedDecreasing=5, interpSpeedIncreasing=5}
    boundBone {boneName=dyn_strap_L_02}
    componentAppliedLinearAccClamp=<800 800 800>
    componentLinearAccScale=<0.2 0.2 0.2>
    componentLinearVelScale=<0.3 0.3 0.3>
    gravityOverride {z=-450}
    physicsBodyDefinitions [{
      boundBone {boneName=dyn_strap_L_02}
      boxExtents=<13 13 13>
      constraintSetup {angularLimitsMax=<2 1 15>, angularLimitsMin=<-2 -3 -1>, angularTarget=<0 0 0>}
      localJointOffset=<1.3 0 0>
     }]
   }
   n31 = Anim Dynamics - Bone: dyn_strap_L_03(ComponentPose=@n30, Alpha=@n29, ExternalForce="0,0,0") {
    angularSpringConstant=1000
    bAngularSpring=true
    bUseGravityOverride=true
    alphaBoolBlend {blendOption=Linear}
    alphaScaleBiasClamp {bInterpResult=true, interpSpeedDecreasing=5, interpSpeedIncreasing=5}
    boundBone {boneName=dyn_strap_L_03}
    componentAppliedLinearAccClamp=<800 800 800>
    componentLinearAccScale=<0.2 0.2 0.2>
    componentLinearVelScale=<0.3 0.3 0.3>
    gravityOverride {z=-500}
    physicsBodyDefinitions [{
      boundBone {boneName=dyn_strap_L_03}
      boxExtents=<13 13 13>
      constraintSetup {angularLimitsMax=<2 5 15>, angularLimitsMin=<-2 -1 -1>, angularTarget=<0 0 0>}
      localJointOffset=<1.3 0 0>
     }]
   }
   n32 = Anim Dynamics - Bone: dyn_strap_L_04(ComponentPose=@n31, Alpha=@n29, ExternalForce="0,0,0") {
    angularSpringConstant=1000
    bAngularSpring=true
    bUseGravityOverride=true
    alphaBoolBlend {blendOption=Linear}
    alphaScaleBiasClamp {bInterpResult=true, interpSpeedDecreasing=5, interpSpeedIncreasing=5}
    boundBone {boneName=dyn_strap_L_04}
    componentAppliedLinearAccClamp=<800 800 800>
    componentLinearAccScale=<0.2 0.2 0.2>
    componentLinearVelScale=<0.3 0.3 0.3>
    gravityOverride {z=-500}
    physicsBodyDefinitions [{
      boundBone {boneName=dyn_strap_L_04}
      boxExtents=<13 13 13>
      constraintSetup {angularLimitsMax=<2 1 5>, angularLimitsMin=<-2 -1 -1>, angularTarget=<0 0 0>}
      localJointOffset=<1.3 0 0>
     }]
   }
   n33 = Anim Dynamics - Bone: dyn_strap_R_base(ComponentPose=@n32, Alpha=@n29, ExternalForce="0,0,0") {
    angularSpringConstant=1000
    bAngularSpring=true
    bUseGravityOverride=true
    alphaBoolBlend {blendOption=Linear}
    alphaScaleBiasClamp {bInterpResult=true, interpSpeedDecreasing=5, interpSpeedIncreasing=5}
    boundBone {boneName=dyn_strap_R_base}
    componentAppliedLinearAccClamp=<800 800 800>
    componentLinearAccScale=<0.2 0.2 0.2>
    componentLinearVelScale=<0.3 0.3 0.3>
    gravityOverride {z=-400}
    physicsBodyDefinitions [{
      boundBone {boneName=dyn_strap_R_base}
      boxExtents=<13 13 13>
      constraintSetup {angularLimitsMax=<2 5 1>, angularLimitsMin=<-2 -10 -30>, angularTarget=<0 0 0>}
      localJointOffset=<1.3 0 0>
     }]
   }
   n34 = Anim Dynamics - Bone: dyn_strap_R_base(ComponentPose=@n33, Alpha=1, ExternalForce="0,0,0") {
    angularSpringConstant=1000
    bAngularSpring=true
    bUseGravityOverride=true
    bUseSphericalLimits=true
    alphaBoolBlend {blendOption=Linear}
    alphaScaleBiasClamp {bInterpResult=true, interpSpeedDecreasing=5, interpSpeedIncreasing=5}
    boundBone {boneName=dyn_strap_R_base}
    componentAppliedLinearAccClamp=<800 800 800>
    componentLinearAccScale=<0.2 0.2 0.2>
    componentLinearVelScale=<0.3 0.3 0.3>
    gravityOverride {z=-400}
    physicsBodyDefinitions [{
      boundBone {boneName=dyn_strap_R_base}
      boxExtents=<13 13 13>
      constraintSetup {angularLimitsMax=<2 5 1>, angularLimitsMin=<-2 0 -30>, angularTarget=<0 0 0>}
      localJointOffset=<1.3 0 0>
     }]
    sphericalLimits [{limitRadius=7, drivingBone {boneName=pack}, sphereLocalOffset=<-7.6413 -8.6377 -18.0085>}]
   }
   n35 = Anim Dynamics - Bone: dyn_strap_R_02(ComponentPose=@n34, Alpha=@n29, ExternalForce="0,0,0") {
    angularSpringConstant=1000
    bAngularSpring=true
    bUseGravityOverride=true
    alphaBoolBlend {blendOption=Linear}
    alphaScaleBiasClamp {bInterpResult=true, interpSpeedDecreasing=5, interpSpeedIncreasing=5}
    boundBone {boneName=dyn_strap_R_02}
    componentAppliedLinearAccClamp=<800 800 800>
    componentLinearAccScale=<0.2 0.2 0.2>
    componentLinearVelScale=<0.3 0.3 0.3>
    gravityOverride {z=-500}
    physicsBodyDefinitions [{
      boundBone {boneName=dyn_strap_R_02}
      boxExtents=<13 13 13>
      constraintSetup {angularLimitsMax=<2 3 1>, angularLimitsMin=<-2 -5 -10>, angularTarget=<0 0 0>}
      localJointOffset=<1.3 0 0>
     }]
   }
   n36 = "Constraint - dyn_strap_R_03 (P)"(ComponentPose=@n35, Alpha=0.3, ConstraintWeights_0=1) {
    alphaBoolBlend {blendOption=Linear}
    boneToModify {boneName=dyn_strap_R_03}
    constraintSetup [{targetBone {boneName=dyn_strap_L_05}}]
    constraintWeights [1]
   }
   n37 = "Constraint - dyn_strap_R_02 (P)"(ComponentPose=@n36, Alpha=0.2, ConstraintWeights_0=1) {
    alphaBoolBlend {blendOption=Linear}
    boneToModify {boneName=dyn_strap_R_02}
    constraintSetup [{targetBone {boneName=dyn_strap_L_05}}]
    constraintWeights [1]
   }
   n38 = "Constraint - dyn_strap_L_05 (P)"(ComponentPose=@n37, Alpha=0.7, ConstraintWeights_0=1) {
    alphaBoolBlend {blendOption=Linear}
    boneToModify {boneName=dyn_strap_L_05}
    constraintSetup [{targetBone {boneName=dyn_strap_R_03}}]
    constraintWeights [1]
   }
   n39 = Component To Local(ComponentPose=@n38)
   n40 = Get bRigidBodyEnabled()
   n41 = Anim Dynamics - Bone: dyn_zipper01_00(
    ComponentPose=Local To Component(LocalPose=Blend Poses by bool(BlendPose_0=@n17, BlendPose_1=@n39, bActiveValue=@n40) {blendType=Linear})
    Alpha=1
    ExternalForce="0,0,0"
   ) {
    angularDampingOverride=0.9999
    angularSpringConstant=600
    bAngularSpring=true
    bOverrideAngularDamping=true
    bUseGravityOverride=true
    alphaBoolBlend {blendOption=Linear}
    boundBone {boneName=dyn_zipper01_00}
    componentAppliedLinearAccClamp=<900 900 900>
    componentLinearAccScale=<0.4 0.4 0.4>
    componentLinearVelScale=<0.5 0.5 0.5>
    gravityOverride {z=-900}
    physicsBodyDefinitions [{
      boundBone {boneName=dyn_zipper01_00}
      boxExtents=<5 5 5>
      constraintSetup {angularLimitsMax=<0 20 0>, angularLimitsMin=<0 0 -20>, angularTarget=<0 0 0>}
      localJointOffset=<1 0 0>
     }]
   }
   n42 = Anim Dynamics - Bone: dyn_zipper01_01(ComponentPose=@n41, Alpha=1, ExternalForce="0,0,0") {
    angularDampingOverride=0.9
    angularSpringConstant=600
    bAngularSpring=true
    bOverrideAngularDamping=true
    bUseGravityOverride=true
    alphaBoolBlend {blendOption=Linear}
    boundBone {boneName=dyn_zipper01_01}
    componentAppliedLinearAccClamp=<800 1200 800>
    componentLinearAccScale=<0.4 0.4 0.4>
    componentLinearVelScale=<0.5 0.5 0.5>
    gravityOverride {z=-900}
    physicsBodyDefinitions [{
      boundBone {boneName=dyn_zipper01_01}
      boxExtents=<5 5 5>
      constraintSetup {angularLimitsMax=<5 40 0>, angularLimitsMin=<-5 0 -20>, angularTarget=<0 0 0>}
      localJointOffset=<0.5 0 0>
     }]
   }
   n43 = Anim Dynamics - Bone: dyn_zipper02_00(ComponentPose=@n42, Alpha=1, ExternalForce="0,0,0") {
    angularDampingOverride=0.9999
    angularSpringConstant=700
    bAngularSpring=true
    bOverrideAngularDamping=true
    bUseGravityOverride=true
    alphaBoolBlend {blendOption=Linear}
    boundBone {boneName=dyn_zipper02_00}
    componentAppliedLinearAccClamp=<800 800 800>
    componentLinearAccScale=<0.2 0.2 0.2>
    componentLinearVelScale=<0.3 0.3 0.3>
    gravityOverride {z=-900}
    physicsBodyDefinitions [{
      boundBone {boneName=dyn_zipper02_00}
      boxExtents=<3.5 3.5 3.5>
      constraintSetup {angularLimitsMax=<10 0 20>, angularLimitsMin=<-10 -35 0>, angularTarget=<0 0 0>}
      localJointOffset=<0.5 0 0>
     }]
   }
   n44 = Anim Dynamics - Bone: dyn_zipper02_01(ComponentPose=@n43, Alpha=1, ExternalForce="0,0,0") {
    angularDampingOverride=0.9999
    angularSpringConstant=700
    bAngularSpring=true
    bOverrideAngularDamping=true
    bUseGravityOverride=true
    alphaBoolBlend {blendOption=Linear}
    boundBone {boneName=dyn_zipper02_01}
    componentAppliedLinearAccClamp=<800 800 800>
    componentLinearAccScale=<0.3 0.3 0.3>
    componentLinearVelScale=<0.4 0.4 0.4>
    gravityOverride {z=-800}
    physicsBodyDefinitions [{
      boundBone {boneName=dyn_zipper02_01}
      boxExtents=<5 5 5>
      constraintSetup {angularLimitsMax=<10 5 20>, angularLimitsMin=<-10 -35 -20>, angularTarget=<0 0 0>}
      localJointOffset=<1 0 0>
     }]
   }
   n45 = "Transform (Modify) Bone - Bone: pack"!(ComponentPose=@n44, bAlphaBoolEnabled:bool=Get bIsSkydiveFloating(), Translation="0,0,0", Rotation="-0.000,0.000,13.532", Scale) -> (Pose) {
    alphaInputType=Bool
    rotationMode=BMM_Additive
    rotationSpace=BCS_BoneSpace
    translationMode=BMM_Additive
    translationSpace=BCS_BoneSpace
    alphaBoolBlend {blendInTime=0.2, blendOption=Linear, blendOutTime=0.2}
    boneToModify {boneName=pack}
    rotation {roll=13.532}
   }
   n46 = Get bIsSkydiveFloating()
   n47 = "Transform (Modify) Bone - Bone: dyn_strap_L_base"!(ComponentPose=@n45, bAlphaBoolEnabled:bool=@n46, Translation="0,0,0", Rotation="-18.824,32.893,20.916", Scale="1,1,1") -> (Pose) {
    alphaInputType=Bool
    rotationMode=BMM_Additive
    rotationSpace=BCS_BoneSpace
    translationMode=BMM_Additive
    translationSpace=BCS_BoneSpace
    alphaBoolBlend {blendInTime=0.2, blendOption=Linear, blendOutTime=0.2}
    boneToModify {boneName=dyn_strap_L_base}
    rotation=<rot -18.824 32.893 20.916>
   }
   n48 = "Transform (Modify) Bone - Bone: dyn_strap_R_base"!(ComponentPose=@n47, bAlphaBoolEnabled:bool=@n46, Translation="0,0,0", Rotation="7.868,-2.154,-14.549", Scale="1,1,1") -> (Pose) {
    alphaInputType=Bool
    rotationMode=BMM_Additive
    rotationSpace=BCS_BoneSpace
    translationMode=BMM_Additive
    translationSpace=BCS_BoneSpace
    alphaBoolBlend {blendInTime=0.2, blendOption=Linear, blendOutTime=0.2}
    boneToModify {boneName=dyn_strap_R_base}
    rotation=<rot 7.868 -2.154 -14.549>
   }
   n49 = "Transform (Modify) Bone - Bone: pack"(ComponentPose=@n48, bAlphaBoolEnabled=Get bIsSkyDiving(), Translation="-0.024,-1.482,0.140", Rotation="10.611,-2.832,4.925") {
    alphaInputType=Bool
    rotationMode=BMM_Additive
    rotationSpace=BCS_BoneSpace
    translationMode=BMM_Additive
    translationSpace=BCS_BoneSpace
    alphaBoolBlend {blendInTime=0.2, blendOption=Linear, blendOutTime=0.2}
    boneToModify {boneName=pack}
    rotation=<rot 10.611 -2.832 4.925>
    translation=<-0.024 -1.482 0.14>
   }
   Output Pose(Result=Component To Local(ComponentPose=@n49)) {layerGroup=None}
   n50 = "Transform (Modify) Bone - Bone: dyn_pompom_lower_01"(
    ComponentPose=Local To Component(LocalPose=Use cached pose 'input pose'())
    Alpha=1
    Translation="0,0,0"
    Rotation="-2.000,0.000,10.000"
    Scale="1,1,1"
   ) {
    rotationMode=BMM_Additive
    rotationSpace=BCS_BoneSpace
    translationMode=BMM_Additive
    translationSpace=BCS_BoneSpace
    alphaBoolBlend {blendOption=Linear}
    boneToModify {boneName=dyn_pompom_lower_01}
    rotation=<rot -2 0 10>
   }
   n51 = "Transform (Modify) Bone - Bone: dyn_pompom_middle_01"(ComponentPose=@n50, Alpha=1, Translation="0,0,0", Rotation="-7,0,10", Scale="1,1,1") {
    rotationMode=BMM_Additive
    rotationSpace=BCS_BoneSpace
    translationMode=BMM_Additive
    translationSpace=BCS_BoneSpace
    alphaBoolBlend {blendOption=Linear}
    boneToModify {boneName=dyn_pompom_middle_01}
    rotation {pitch=-7, roll=10}
   }
   n52 = "Transform (Modify) Bone - Bone: dyn_pompom_top_00"(ComponentPose=@n51, Alpha=1, Translation="0,0,0", Rotation="0,-10,0", Scale="1,1,1") {
    rotationMode=BMM_Additive
    rotationSpace=BCS_BoneSpace
    translationMode=BMM_Additive
    translationSpace=BCS_BoneSpace
    alphaBoolBlend {blendOption=Linear}
    boneToModify {boneName=dyn_pompom_top_00}
    rotation {yaw=-10}
   }
   n53 = "Transform (Modify) Bone - Bone: dyn_pompom_top_01"(ComponentPose=@n52, Alpha=1, Translation="0,0,0", Rotation="0,0,7", Scale="1,1,1") {
    rotationMode=BMM_Additive
    rotationSpace=BCS_BoneSpace
    translationMode=BMM_Additive
    translationSpace=BCS_BoneSpace
    alphaBoolBlend {blendOption=Linear}
    boneToModify {boneName=dyn_pompom_top_01}
    rotation {roll=7}
   }
   n54 = "Transform (Modify) Bone - Bone: dyn_pompom_middle_00"(ComponentPose=@n53, Alpha=1, Translation="0,0,0", Rotation="-10,-10,0", Scale="1,1,1") {
    rotationMode=BMM_Additive
    rotationSpace=BCS_BoneSpace
    translationMode=BMM_Additive
    translationSpace=BCS_BoneSpace
    alphaBoolBlend {blendOption=Linear}
    boneToModify {boneName=dyn_pompom_middle_00}
    rotation {pitch=-10, yaw=-10}
   }
   n55 = "Transform (Modify) Bone - Bone: dyn_pompom_middle_02"(ComponentPose=@n54, Alpha=1, Translation="0,0,0", Rotation="5,0,0", Scale="1,1,1") {
    rotationMode=BMM_Additive
    rotationSpace=BCS_BoneSpace
    translationMode=BMM_Additive
    translationSpace=BCS_BoneSpace
    alphaBoolBlend {blendOption=Linear}
    boneToModify {boneName=dyn_pompom_middle_02}
    rotation {pitch=5}
   }
   n56 = Anim Dynamics - Bone: dyn_pompom_lower_00(ComponentPose=@n55, Alpha=1, ExternalForce="0,0,0") {
    angularDampingOverride=0.9
    angularSpringConstant=2000
    bOverrideAngularDamping=true
    bUseGravityOverride=true
    simulationSpace=World
    alphaBoolBlend {blendOption=Linear}
    boundBone {boneName=dyn_pompom_lower_00}
    componentAppliedLinearAccClamp=<2000 2000 2000>
    componentLinearAccScale=<0.3 0.3 0.3>
    componentLinearVelScale=<0.4 0.4 0.4>
    gravityOverride {z=-980}
    physicsBodyDefinitions [{
      boundBone {boneName=dyn_pompom_lower_00}
      boxExtents=<5 5 5>
      constraintSetup {angularTargetAxis=AxisZ, angularLimitsMax=<15 15 0>, angularLimitsMin=<-15 -10 0>, angularTarget=<0 0 1>}
      localJointOffset=<0 0 0>
     }]
    relativeSpaceBone {boneName=pack}
   }
   n57 = Anim Dynamics - Bone: dyn_pompom_lower_01(ComponentPose=@n56, Alpha=1, ExternalForce="0,0,0") {
    angularDampingOverride=0.9
    bOverrideAngularDamping=true
    bUseGravityOverride=true
    simulationSpace=World
    alphaBoolBlend {blendOption=Linear}
    boundBone {boneName=dyn_pompom_lower_01}
    componentAppliedLinearAccClamp=<1000 1000 1000>
    gravityOverride {z=-980}
    physicsBodyDefinitions [{
      boundBone {boneName=dyn_pompom_lower_01}
      boxExtents=<15 15 15>
      constraintSetup {angularTargetAxis=AxisZ, angularLimitsMax=<20 20 5>, angularLimitsMin=<-20 -5 0>, angularTarget=<0 0 1>}
      localJointOffset=<0.2 0 0>
     }]
   }
   n58 = Anim Dynamics - Bone: dyn_pompom_lower_02(ComponentPose=@n57, Alpha=1, ExternalForce="0,0,0") {
    angularDampingOverride=0.9
    bOverrideAngularDamping=true
    bUseGravityOverride=true
    simulationSpace=World
    alphaBoolBlend {blendOption=Linear}
    boundBone {boneName=dyn_pompom_lower_02}
    componentAppliedLinearAccClamp=<1000 1000 1000>
    gravityOverride {z=-980}
    physicsBodyDefinitions [{
      boundBone {boneName=dyn_pompom_lower_02}
      boxExtents=<15 15 15>
      constraintSetup {angularTargetAxis=AxisZ, angularLimitsMax=<20 15 0>, angularLimitsMin=<-20 -10 -5>, angularTarget=<0 0 1>}
      localJointOffset=<-0.2 0 0>
     }]
   }
   n59 = Anim Dynamics - Bone: dyn_pompom_middle_00(ComponentPose=@n58, Alpha=1, ExternalForce="0,0,0") {
    angularDampingOverride=0.9
    bOverrideAngularDamping=true
    bUseGravityOverride=true
    simulationSpace=World
    alphaBoolBlend {blendOption=Linear}
    boundBone {boneName=dyn_pompom_middle_00}
    componentAppliedLinearAccClamp=<1000 1000 1000>
    gravityOverride {z=-980}
    physicsBodyDefinitions [{
      boundBone {boneName=dyn_pompom_middle_00}
      boxExtents=<10 10 10>
      constraintSetup {angularTargetAxis=AxisZ, angularLimitsMax=<20 40 0>, angularLimitsMin=<-20 -20 -10>, angularTarget=<0 0 1>}
      localJointOffset=<0.3 0 0>
     }]
   }
   n60 = Anim Dynamics - Bone: dyn_pompom_middle_01(ComponentPose=@n59, Alpha=1, ExternalForce="0,0,0") {
    angularDampingOverride=0.9
    bOverrideAngularDamping=true
    bUseGravityOverride=true
    simulationSpace=World
    alphaBoolBlend {blendOption=Linear}
    boundBone {boneName=dyn_pompom_middle_01}
    componentAppliedLinearAccClamp=<1000 1000 1000>
    gravityOverride {z=-980}
    physicsBodyDefinitions [{
      boundBone {boneName=dyn_pompom_middle_01}
      boxExtents=<10 10 10>
      constraintSetup {angularTargetAxis=AxisZ, angularLimitsMax=<20 20 5>, angularLimitsMin=<-20 -10 0>, angularTarget=<0 0 1>}
      localJointOffset=<0.3 0 0>
     }]
   }
   n61 = Anim Dynamics - Bone: dyn_pompom_middle_02(ComponentPose=@n60, Alpha=1, ExternalForce="0,0,0") {
    angularDampingOverride=0.9
    bOverrideAngularDamping=true
    bUseGravityOverride=true
    simulationSpace=World
    alphaBoolBlend {blendOption=Linear}
    boundBone {boneName=dyn_pompom_middle_02}
    componentAppliedLinearAccClamp=<1000 1000 1000>
    gravityOverride {z=-980}
    physicsBodyDefinitions [{
      boundBone {boneName=dyn_pompom_middle_02}
      boxExtents=<10 10 10>
      constraintSetup {angularTargetAxis=AxisZ, angularLimitsMax=<20 30 5>, angularLimitsMin=<-20 -5 0>, angularTarget=<0 0 1>}
      localJointOffset=<-0.1 0 0>
     }]
   }
   n62 = Anim Dynamics - Bone: dyn_pompom_top_00(ComponentPose=@n61, Alpha=1, ExternalForce="0,0,0") {
    angularDampingOverride=0.9
    bOverrideAngularDamping=true
    bUseGravityOverride=true
    simulationSpace=World
    alphaBoolBlend {blendOption=Linear}
    boundBone {boneName=dyn_pompom_top_00}
    componentAppliedLinearAccClamp=<1000 1000 1000>
    gravityOverride {z=-980}
    physicsBodyDefinitions [{
      boundBone {boneName=dyn_pompom_top_00}
      boxExtents=<6 6 6>
      constraintSetup {angularTargetAxis=AxisZ, angularLimitsMax=<20 50 0>, angularLimitsMin=<-20 -15 -10>, angularTarget=<0 0 1>}
      localJointOffset=<1 0 0>
     }]
   }
   n63 = Anim Dynamics - Bone: dyn_pompom_top_01(ComponentPose=@n62, Alpha=1, ExternalForce="0,0,0") {
    angularDampingOverride=0.999
    bOverrideAngularDamping=true
    bUseGravityOverride=true
    simulationSpace=World
    alphaBoolBlend {blendOption=Linear}
    boundBone {boneName=dyn_pompom_top_01}
    componentAppliedLinearAccClamp=<1000 1000 1000>
    gravityOverride {z=-980}
    physicsBodyDefinitions [{
      boundBone {boneName=dyn_pompom_top_01}
      boxExtents=<6 6 6>
      constraintSetup {angularTargetAxis=AxisZ, angularLimitsMax=<20 20 0>, angularLimitsMin=<-20 -5 -5>, angularTarget=<0 0 1>}
      localJointOffset=<0.05 0 0>
     }]
    relativeSpaceBone {boneName=pack}
   }
   n64 = Anim Dynamics - Bone: dyn_pompom_top_02(ComponentPose=@n63, Alpha=1, ExternalForce="0,0,0") {
    angularDampingOverride=0.9
    bOverrideAngularDamping=true
    bUseGravityOverride=true
    simulationSpace=World
    alphaBoolBlend {blendOption=Linear}
    boundBone {boneName=dyn_pompom_top_02}
    componentAppliedLinearAccClamp=<1000 1000 1000>
    gravityOverride {z=-980}
    physicsBodyDefinitions [{
      boundBone {boneName=dyn_pompom_top_02}
      boxExtents=<6 6 6>
      constraintSetup {angularTargetAxis=AxisZ, angularLimitsMax=<20 5 2>, angularLimitsMin=<-20 -30 -2>, angularTarget=<0 0 1>}
      localJointOffset=<-0.7 0 0>
     }]
   }
   Save cached pose 'FE fix'(Pose=Component To Local(ComponentPose=@n64)) {cacheName=FE fix}
   Save cached pose 'input pose'(Pose=Input Pose()) {cacheName=input pose}
  ]}
 EventGraph {nodes [
   Event BlueprintUpdateAnimation()
   > Set scalability alpha(scalability alpha:double=SelectFloat(A=0.5, B=1, bPickA=Get bIsSkyDiving())) -> (Output_Get:double)
   > n1 = Branch(Condition=OR Boolean(A=Get bIsSkydiveFloating(), B=Get bIsSkyDiving()))
   > Set Gravity Override(Gravity Override="0,-300,600")
   Set Gravity Override(execute=@n1.else, Gravity Override="0,0,-980")
   TryGetPawnOwner()
  ]}
]
variables [
 Gravity Override {type=Vector}
 scalability alpha {type=number}
]
